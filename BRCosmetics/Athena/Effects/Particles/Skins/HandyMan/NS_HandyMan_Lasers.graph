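# Niagara system for the HandyMan cosmetic skin that attaches red glowing lasers to the character's skeletal mesh.
# The Lasers emitter bursts two particles after a short delay, pins each to a bone or socket by spawn group, and renders oriented meshes with offset glow sprites.
# LaserLight adds a single red light, re-sampled from the skeleton each update and offset along its orientation.
# Particles persist rather than expiring by lifetime.
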NiagaraSystem /BRCosmetics/Athena/Effects/Particles/Skins/HandyMan/NS_HandyMan_Lasers
dependencies {usedDataInterfaces [], usedDynamicInputs [], usedModules [&SystemState]}
emitters {
 LaserLight {
  emitterUpdateScript [
   &EmitterState(Life Cycle Mode=::NewEnumerator1, Loop Behavior=::NewEnumerator1, Loop Duration Mode=::NewEnumerator1, UseLoopDelay=false)
   &SpawnBurst_Instantaneous(Spawn Count=1, Spawn Time=0.4, Use Spawn Probability=false)
  ]
  particleSpawnScript [/Niagara/Modules/Spawn/Initialization/InitializeParticle(
    Color=<rgba 1.5 0 0.043 1>
    Lifetime=0
    Module.Write Color=true
    Module.Write Lifetime=true
    Module.Write Mass=false
    Module.Write Material Random=true
    Module.Write Position=false
    Module.Write Scale=true
    Module.Write SpriteRotation=false
    Module.Write SpriteSize=false
    Position=&SimulationPosition()
    Sprite Rotation=180
   )]
  particleUpdateScript [
   &SampleSkeletalMeshSkeleton(
    Mesh:NiagaraDataInterfaceSkeletalMesh=$User.SkeletalMesh
    Spawn Mode:ENiagaraSkeletalSpawning=::NewEnumerator1
    Specific Socket Index:NiagaraInt32=&MakeIntFromFloat(Float=0)
   )
   /Niagara/Modules/Spawn/Location/SkeletalMeshSkeletonLocation(Module.Inherit Velocity=false, Module.Use Sampled Bone as Ribbon ID=true)
   /Game/Effects/Niagara/Modules/Position/FN_OffsetPositionByOrientationAxis(Offset Axis:Vector3f=<-24 0 -2.5>) {module=OffsetPositionByMeshOrientation}
   &FN_Light_v2(Light Exponent=12, Light Radius=30, LightEnabled=true)
   /SetVariables() {isSetParameters=true}
   &ParticleState(Kill Particles When Lifetime Has Elapsed=false, Let Infinitely Lived Particles Die When Emitter Deactivates=true)
   &SolveForcesAndVelocity(Clamp Velocity=false, Limit Acceleration=false)
  ]
  rendererClasses [&NiagaraLightRendererProperties]
  renderers [{rendererClass=&NiagaraLightRendererProperties, rendererIndex=0}]
 }
 Lasers {
  emitterUpdateScript [
   &EmitterState(Life Cycle Mode=::NewEnumerator1, Loop Behavior=::NewEnumerator1, Loop Duration Mode=::NewEnumerator1, UseLoopDelay=false)
   &SpawnBurst_Instantaneous(Spawn Count=2, Spawn Time=0.4, Use Spawn Probability=false)
  ]
  particleSpawnScript [
   &InitializeParticle(
    Color=<rgba 5 0 0.1432 1>
    Mesh Scale=<0.2 0.05 0.05>
    Module.Write Color=true
    Module.Write Lifetime=true
    Module.Write Mass=false
    Module.Write Material Random=true
    Module.Write Position=false
    Module.Write Scale=true
    Module.Write SpriteRotation=false
    Module.Write SpriteSize=false
    Position=&SimulationPosition()
    Sprite Rotation=180
   )
   &DynamicMaterialParameters(
    Index 0 Param 1=100
    Module.Param0WriteEnabled=true
    Module.Param1WriteEnabled=true
    Module.Param2WriteEnabled=true
    Module.Param3WriteEnabled=true
   )
   /Game/Effects/Niagara/Modules/Groups/FN_SetSpawnGroup_v1/FN_SetSpawnGroup(Emitter Bone Count=2)
   /SetVariables(Particles.GlowSpriteColor:LinearColor=<rgba 50 2.5461 1.5513 5>, Particles.GlowSpriteSize:Vector2f=<100 100>) {isSetParameters=true}
  ]
  particleUpdateScript [
   &SkeletalMeshLocation(
    Apply Sampled Orientation to Mesh Particles=true
    Bone / Socket Index=$Particles.SpawnGroup
    Inherit Velocity=false
    MaskSpawnGroup=false
    Mesh Sampling Type=::NewEnumerator5
    Skeletal Mesh=$User.SkeletalMesh
    Socket Sampling Mode=::NewEnumerator5
   )
   /SetVariables(Particles.GlowSpriteOffset:Vector3f=&Add_Vector(
     A=Linked() {linkedVariable {name=Particles.Position, type=NiagaraPosition}}
     B=&RotateVectorByQuaternion(Vector To Rotate=<-4 0 0>, Quaternion=$Output.SkeletalMeshLocation.SampledRotation)
    )) {isSetParameters=true}
   &ParticleState(Kill Particles When Lifetime Has Elapsed=false, Let Infinitely Lived Particles Die When Emitter Deactivates=true)
   &SolveForcesAndVelocity(Clamp Velocity=false, Limit Acceleration=false)
  ]
  rendererClasses [&NiagaraMeshRendererProperties, &NiagaraSpriteRendererProperties]
  renderers [
   {rendererClass=&NiagaraMeshRendererProperties, rendererIndex=0}
   {rendererClass=&NiagaraSpriteRendererProperties, rendererIndex=1}
  ]
 }
}
summary {
 systemName=NS_HandyMan_Lasers
 emitters [
  {emitterName=Lasers, rendererClasses [&NiagaraMeshRendererProperties, &NiagaraSpriteRendererProperties]}
  {emitterName=LaserLight, rendererClasses [&NiagaraLightRendererProperties]}
 ]
 userVariables [
  User.bIsFrontEndPreview {type=NiagaraBool}
  User.SkeletalMesh {
   type=NiagaraDataInterfaceSkeletalMesh
   defaultValue {
    struct=NiagaraExt_VariableValue_DataInterface
    value {dataInterface=&self:NiagaraDataInterfaceSkeletalMesh_0}
   }
  }
 ]
}
